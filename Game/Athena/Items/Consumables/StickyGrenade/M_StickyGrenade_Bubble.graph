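# Additive, two-sided mesh particle material for the Sticky Grenade bubble effect, drawing a glowing blue energy shell.
# Two counter-panned fire noises, mapped by camera vector, view-space normal and time, blend Color1 to Color2, with dynamic parameters adjusting tiling and offset.
# A panned noise masked by inverted Fresnel adds detail, then particle and vertex color scale the emissive.
# Opacity comes from Fresnel times particle alpha, softened by depth fade.
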Material /Game/Athena/Items/Consumables/StickyGrenade/M_StickyGrenade_Bubble
graph {
 properties {bUsedWithMeshParticles=true, blendMode=BLEND_Additive, pixelDepthOffsetMode=PDOM_Legacy, twoSided=true, usedShadingModels=Default Lit}
 nodes [
  n1 = Multiply(A=MaterialFunctionCall(Exponent=Constant() {r=8}) {materialFunction=&MF_Fresnel_FX}.Fresnel) {constB=8}
  n2 = ParticleColor()
  DepthFade(Opacity=Clamp(None=Multiply(A=@n1, B=@n2.A)), FadeDistance=ScalarParameter() {defaultValue=150, parameterName=FadeDistance})
  n3 = TextureCoordinate() {uTiling=2}
  n4 = Power(Base=TextureSample(UVs=Panner(Coordinate=@n3) {speedX=0.65, speedY=0.5}) {texture=&T_NoisePack_01}.G)
  n5 = ComponentMask(None=Constant4Vector() {constant=<rgba 0.5 0.5 0.5 0>}.RGBA) {b=true, g=true, r=true}
  n6 = MaterialFunctionCall(Exponent=Constant() {r=8}) {materialFunction=&MF_Fresnel_FX}
  n7 = VectorParameter() {parameterName=Color1, defaultValue=<rgba 0.0048 0.0055 0.06 0>}
  n8 = Multiply(A=Time() {period=3}) {constB=0.3}
  n9 = PixelNormalWS()
  n10 = Add(
   A=Add(A=CameraVectorWS(), B=MaterialFunctionCall(G=@n8) {materialFunction=&AppendMany}.RGB)
   B=Multiply(
    A=Transform(None=@n9) {transformSourceType=TRANSFORMSOURCE_World, transformType=TRANSFORM_View}
    B=ScalarParameter() {defaultValue=0.5, parameterName=Normal Strength}
   )
  )
  n11 = DynamicParameter()
  n12 = Multiply(A=VectorParameter() {parameterName=Tiling, defaultValue=<rgba 0.25 0.25 0 1>}.RGB, B=@n11.Param2)
  n13 = Add(A=Multiply(A=ComponentMask(None=@n10) {g=true, r=true}, B=ComponentMask(None=@n12) {g=true, r=true}), B=@n11.Param1)
  n14 = Add(
   A=TextureSample(UVs=Panner(Coordinate=@n13) {speedX=0.12, speedY=-0.15}) {texture=/Game/Global/Textures/Default/Noise/T_FireNoise_01}.G
   B=TextureSample(UVs=Panner(Coordinate=@n13) {speedX=-0.1, speedY=0.1}) {texture=&T_FireNoise_01}.G
  )
  n15 = LinearInterpolate(A=@n7.RGB, B=VectorParameter() {parameterName=Color2, defaultValue=<rgba 0.0617 0.6535 0.9688 0>}.RGB, Alpha=@n14)
  n16 = Power(Base=@n14, Exp=ScalarParameter() {defaultValue=8, parameterName=Exponent}) {constExponent=4}
  n17 = Add(A=Multiply(A=Multiply(A=@n4, B=@n5), B=Multiply(A=@n6.Fresnel Inverted) {constB=4}), B=Add(A=@n7.RGB, B=Multiply(A=@n15, B=@n16)))
  n18 = VertexColor()
  MaterialFunctionCall(Emissive=Multiply(A=ParticleColor().RGB, B=Multiply(A=@n17, B=@n18.A))) {materialFunction=&MF_FX_EmissiveAdjustments}
 ]
}
