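# Masked, two-sided mesh-particle material for the sticky grenade's shock bubble effect.
# Two channels of a muzzle texture are sampled at separate random offsets and summed, with dynamic parameters driving tiling and panning and a Fresnel term bending the UVs.
# Particle color feeds emissive adjustments, with back faces at full color and front faces dimmed by half except where Fresnel restores them.
# A vector parameter supplies a world position offset value.
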Material /Game/Athena/Items/Consumables/StickyGrenade/M_StickyGrenade_ShockBubble
graph {
 properties {
  bUsedWithMeshParticles=true
  blendMode=BLEND_Masked
  pixelDepthOffsetMode=PDOM_Legacy
  translucencyLightingMode=TLM_VolumetricDirectional
  twoSided=true
  usedShadingModels=Default Lit
 }
 nodes [
  n1 = MaterialFunctionCall(Float2=TextureCoordinate()) {materialFunction=&BreakOutFloat2Components}
  n2 = DynamicParameter() {defaultValue=<rgba 1 0 1 0.1526>, paramNames [U Tile, Random A, Random B, Panner U]}
  n3 = MaterialFunctionCall(Exponent=ScalarParameter() {defaultValue=1, parameterName=Bend Exponent}) {materialFunction=&MF_Fresnel_FX}
  n4 = AppendVector(
   A=Multiply(A=DynamicParameter() {defaultValue=<rgba 1 0 1 0.1526>, paramNames [U Tile, Random A, Random B, Panner U]}.U Tile, B=@n1.R)
   B=Add(A=Add(A=@n1.G, B=@n2.Panner U), B=ConstantBiasScale(None=@n3.Fresnel) {bias=0, scale=-0.5})
  )
  n5 = DynamicParameter() {defaultValue=<rgba 1 0 1 0.1526>, paramNames [U Tile, Random A, Random B, Panner U]}
  n6 = AppendVector(A=@n5.Random B, B=Constant())
  Add(
   A=TextureSample(UVs=Add(A=@n4, B=@n5.Random A)) {samplerType=SAMPLERTYPE_LinearColor, texture=/Game/Effects/Fort_Effects/Textures/Tile/T_Raygun_Muzzle_01}.R
   B=TextureSample(UVs=Add(A=@n4, B=@n6)) {samplerType=SAMPLERTYPE_LinearColor, texture=&T_Raygun_Muzzle_01}.B
  )
  n7 = ParticleColor()
  n8 = Multiply(B=@n7.RGB) {constA=0.5}
  n9 = MaterialFunctionCall(
   Brightness=ScalarParameter() {defaultValue=1, parameterName=Fresnel Brightness}
   Exponent=ScalarParameter() {defaultValue=1, parameterName=Fresnel Exponent}
  ) {materialFunction=&MF_Fresnel_FX}
  n10 = TwoSidedSign()
  MaterialFunctionCall(Emissive=LinearInterpolate(A=@n7.RGB, B=LinearInterpolate(A=@n8, B=@n7.RGB, Alpha=@n9.Fresnel), Alpha=ConstantBiasScale(None=@n10))) {materialFunction=&MF_FX_EmissiveAdjustments}
  VectorParameter() {parameterName=WPO Offset, defaultValue=<rgba 0 0 60 0>}
 ]
}
